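# Animation Blueprint for the M_MED_Elastic CAU_1 part mesh that adds secondary hair motion on top of the pose copied from the attached parent.
# Head twist relative to spine_05 drives an additive offset on hair bone C, ahead of chained spring-based AnimDynamics on bones A, B (low weight) and C.
# Skydive diving or floating zeroes backHairSimAlpha, turning off C's simulation and, through an inverted alpha range, blending in a fixed additive offset instead.
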AnimBlueprint /Game/Characters/Player/Male/Medium/Bodies/M_MED_Elastic/Meshes/Parts/CAU_Parts/M_MED_Elastic_CAU_1_AnimBP
target_skeleton=&/M_MED_Elastic_CAU_1_Skeleton
graphs [
 AnimGraph {nodes [
   n1 = "Transform (Modify) Bone - Bone: dyn_hairCAU1_C_01_r"(
    ComponentPose=Local To Component(LocalPose=Copy Pose From Mesh() {bUseAttachedParent=true})
    Alpha=Get headAlpha() -> (headAlpha:double)
    Translation="-0.75,-0.255,1.5"
    Rotation="-29.834,0.236,-0.745"
    Scale="1,1,1"
   ) {
    rotationMode=BMM_Additive
    rotationSpace=BCS_BoneSpace
    translationMode=BMM_Additive
    translationSpace=BCS_BoneSpace
    alphaBoolBlend {blendOption=Linear}
    boneToModify {boneName=dyn_hairCAU1_C_01_r}
    rotation=<rot -29.8341 0.236 -0.745>
    translation=<-0.75 -0.255 1.5>
   }
   n2 = Anim Dynamics - Bone: dyn_hairCAU1_A_01_r(ComponentPose=@n1, Alpha=1, ExternalForce="0,0,0") {
    angularDampingOverride=1
    angularSpringConstant=1500
    bAngularSpring=true
    bGravityOverrideInSimSpace=true
    bOverrideAngularDamping=true
    gravityScale=0.35
    simulationSpace=RootRelative
    alphaBoolBlend {blendOption=Linear}
    boundBone {boneName=dyn_hairCAU1_A_01_r}
    componentLinearAccScale=<0.6 0.6 0.6>
    componentLinearVelScale=<0.6 0.6 0.6>
    physicsBodyDefinitions [{
      sphereCollisionRadius=0
      boundBone {boneName=dyn_hairCAU1_A_01_r}
      boxExtents=<10 20 20>
      constraintSetup {angularLimitsMax=<0 0 0>, angularLimitsMin=<0 -40 -5>, angularTarget=<0 0 0>}
      localJointOffset=<5 0 0>
     }]
   }
   n3 = Anim Dynamics - Bone: dyn_hairCAU1_B_01_r(ComponentPose=@n2, Alpha=0.1, ExternalForce="0,0,0") {
    angularDampingOverride=1
    angularSpringConstant=1500
    bAngularSpring=true
    bGravityOverrideInSimSpace=true
    bOverrideAngularDamping=true
    gravityScale=0.35
    simulationSpace=RootRelative
    alphaBoolBlend {blendOption=Linear}
    boundBone {boneName=dyn_hairCAU1_B_01_r}
    componentLinearAccScale=<0.6 0.6 0.6>
    componentLinearVelScale=<0.6 0.6 0.6>
    physicsBodyDefinitions [{
      sphereCollisionRadius=0
      boundBone {boneName=dyn_hairCAU1_B_01_r}
      boxExtents=<10 20 20>
      constraintSetup {angularLimitsMax=<0 5 0>, angularLimitsMin=<0 -40 0>, angularTarget=<0 0 0>}
      localJointOffset=<5 0 0>
     }]
   }
   n4 = Anim Dynamics - Bone: dyn_hairCAU1_C_01_r(ComponentPose=@n3, Alpha=Get backHairSimAlpha() -> (backHairSimAlpha:double), ExternalForce="0,0,0") {
    angularDampingOverride=1
    angularSpringConstant=1500
    bAngularSpring=true
    bGravityOverrideInSimSpace=true
    bOverrideAngularDamping=true
    gravityScale=0.35
    simulationSpace=RootRelative
    alphaBoolBlend {blendOption=Linear}
    alphaScaleBiasClamp {bMapRange=true, outRange {max=0.5}}
    boundBone {boneName=dyn_hairCAU1_C_01_r}
    componentLinearAccScale=<0.6 0.6 0.6>
    componentLinearVelScale=<0.6 0.6 0.6>
    physicsBodyDefinitions [{
      sphereCollisionRadius=0
      boundBone {boneName=dyn_hairCAU1_C_01_r}
      boxExtents=<10 20 20>
      constraintSetup {angularLimitsMax=<0 0 5>, angularLimitsMin=<0 -40 -5>, angularTarget=<0 0 0>}
      localJointOffset=<5 0 0>
     }]
   }
   n5 = "Transform (Modify) Bone - Bone: dyn_hairCAU1_C_01_r"(ComponentPose=@n4, Alpha=Get backHairSimAlpha(), Translation="0,0,-1", Rotation="-20,0,0", Scale="1,1,1") {
    rotationMode=BMM_Additive
    rotationSpace=BCS_BoneSpace
    translationMode=BMM_Additive
    translationSpace=BCS_BoneSpace
    alphaBoolBlend {blendOption=Linear}
    alphaScaleBiasClamp {bMapRange=true, outRange {max=0, min=1}}
    boneToModify {boneName=dyn_hairCAU1_C_01_r}
    rotation {pitch=-20}
    translation {z=-1}
   }
   Output Pose(Result=Component To Local(ComponentPose=@n5)) {layerGroup=None}
  ]}
 EventGraph {nodes [
   Event BlueprintUpdateAnimation()
   n1 = GetAlphaFromReferenceTransformDelta(
    SkeletalMeshComponent=GetPartSkeletalMeshComponent()
    BoneName=head
    BaseName=spine_05
    DeltaComponentType=QuaternionTwist
    InputRange=(X=0,Y=30)
    TwistAxis="0,0,1"
   )
   > Set headAlpha(headAlpha:double=@n1) -> (Output_Get:double)
   > n2 = Branch(Condition=OR Boolean(A=Get bIsSkydiveDiveMode(), B=Get bIsSkydiveFloating()))
   > Set backHairSimAlpha(backHairSimAlpha:double=0) -> (Output_Get:double)
   Set backHairSimAlpha(execute=@n2.else, backHairSimAlpha=1)
   SelectFloat(A=@n1, B=0, bPickA=Get bIsTargeting())
  ]}
]
variables [
 backHairSimAlpha {type=number}
 headAlpha {type=number}
]
